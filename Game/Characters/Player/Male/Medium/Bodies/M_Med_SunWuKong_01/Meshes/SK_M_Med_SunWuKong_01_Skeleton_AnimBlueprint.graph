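# Animation Blueprint for the Sun Wukong player body that copies the pose from the owner's mesh and adds shoulder armor correctives and secondary motion.
# Curve-driven Bone Driven Controllers move both pads from upperarm rotation during skydiving or parachuting, at half strength with an umbrella.
# Otherwise, a twist-driven rotation and additive offset correct only the left pad.
# Anim Dynamics then jiggles both pads, with alpha scaled by 2D pawn speed.
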AnimBlueprint /Game/Characters/Player/Male/Medium/Bodies/M_Med_SunWuKong_01/Meshes/SK_M_Med_SunWuKong_01_Skeleton_AnimBlueprint
target_skeleton=&/SK_M_Med_SunWuKong_01_Skeleton
graphs [
 AnimGraph {nodes [
   n1 = Get MeshToCopy() -> (MeshToCopy:Skeletal Mesh Component Object Reference)
   n2 = Get InAirShoulderCorrectiveAlpha() -> (InAirShoulderCorrectiveAlpha:double)
   n3 = "shoulderArmor_l_offset.translateX = curve(upperarm_l.rotateY) - Bone Driven Controller"!(
    ComponentPose:Component Space Pose Link Structure=Local To Component(LocalPose=Copy Pose From Mesh(SourceMeshComponent=@n1))
    Alpha:float=@n2
   ) -> (Pose:Component Space Pose Link Structure) {
    is=title AnimGraphNode_BoneDrivenController
    bAffectTargetTranslationX=true
    drivingCurve=&/wuKong_shoulderPad_l_translate_x
    lODThreshold=2
    sourceComponent=RotationY
    alphaBoolBlend {blendOption=Linear}
    sourceBone {boneName=upperarm_l}
    targetBone {boneName=shoulderArmor_l_offset}
   }
   n4 = "shoulderArmor_l_offset.translateY = curve(upperarm_l.rotateY) - Bone Driven Controller"!(ComponentPose:Component Space Pose Link Structure=@n3, Alpha:float=@n2) -> (Pose:Component Space Pose Link Structure) {
    is=title AnimGraphNode_BoneDrivenController
    bAffectTargetTranslationY=true
    drivingCurve=&/wuKong_shoulderPad_l_translate_y
    lODThreshold=2
    sourceComponent=RotationY
    alphaBoolBlend {blendOption=Linear}
    sourceBone {boneName=upperarm_l}
    targetBone {boneName=shoulderArmor_l_offset}
   }
   n5 = "shoulderArmor_l_offset.translateZ = curve(upperarm_l.rotateY) - Bone Driven Controller"!(ComponentPose:Component Space Pose Link Structure=@n4, Alpha:float=@n2) -> (Pose:Component Space Pose Link Structure) {
    is=title AnimGraphNode_BoneDrivenController
    bAffectTargetTranslationZ=true
    drivingCurve=&/wuKong_shoulderPad_l_translate_z
    lODThreshold=2
    sourceComponent=RotationY
    alphaBoolBlend {blendOption=Linear}
    sourceBone {boneName=upperarm_l}
    targetBone {boneName=shoulderArmor_l_offset}
   }
   n6 = "shoulderArmor_l_offset.rotateY = curve(upperarm_l.rotateY) - Bone Driven Controller"!(ComponentPose:Component Space Pose Link Structure=@n5, Alpha:float=@n2) -> (Pose:Component Space Pose Link Structure) {
    is=title AnimGraphNode_BoneDrivenController
    bAffectTargetRotationY=true
    drivingCurve=&/wuKong_shoulderPad_l_rotate
    lODThreshold=2
    sourceComponent=RotationY
    alphaBoolBlend {blendOption=Linear}
    sourceBone {boneName=upperarm_l}
    targetBone {boneName=shoulderArmor_l_offset}
   }
   n7 = "shoulderArmor_r_offset.translateX = curve(upperarm_r.rotateY) - Bone Driven Controller"!(ComponentPose:Component Space Pose Link Structure=@n6, Alpha:float=@n2) -> (Pose:Component Space Pose Link Structure) {
    is=title AnimGraphNode_BoneDrivenController
    bAffectTargetTranslationX=true
    drivingCurve=&/wuKong_shoulderPad_r_translate_x
    lODThreshold=2
    sourceComponent=RotationY
    alphaBoolBlend {blendOption=Linear}
    sourceBone {boneName=upperarm_r}
    targetBone {boneName=shoulderArmor_r_offset}
   }
   n8 = "shoulderArmor_r_offset.translateY = curve(upperarm_r.rotateY) - Bone Driven Controller"!(ComponentPose:Component Space Pose Link Structure=@n7, Alpha:float=@n2) -> (Pose:Component Space Pose Link Structure) {
    is=title AnimGraphNode_BoneDrivenController
    bAffectTargetTranslationY=true
    drivingCurve=&/wuKong_shoulderPad_r_translate_y
    lODThreshold=2
    sourceComponent=RotationY
    alphaBoolBlend {blendOption=Linear}
    sourceBone {boneName=upperarm_r}
    targetBone {boneName=shoulderArmor_r_offset}
   }
   n9 = "shoulderArmor_r_offset.translateZ = curve(upperarm_r.rotateY) - Bone Driven Controller"!(ComponentPose:Component Space Pose Link Structure=@n8, Alpha:float=@n2) -> (Pose:Component Space Pose Link Structure) {
    is=title AnimGraphNode_BoneDrivenController
    bAffectTargetTranslationZ=true
    drivingCurve=&/wuKong_shoulderPad_r_translate_z
    lODThreshold=2
    sourceComponent=RotationY
    alphaBoolBlend {blendOption=Linear}
    sourceBone {boneName=upperarm_r}
    targetBone {boneName=shoulderArmor_r_offset}
   }
   n10 = "shoulderArmor_r_offset.rotateY = curve(upperarm_r.rotateY) - Bone Driven Controller"!(ComponentPose:Component Space Pose Link Structure=@n9, Alpha:float=@n2) -> (Pose:Component Space Pose Link Structure) {
    is=title AnimGraphNode_BoneDrivenController
    bAffectTargetRotationY=true
    drivingCurve=&/wuKong_shoulderPad_r_rotate
    lODThreshold=2
    sourceComponent=RotationY
    alphaBoolBlend {blendOption=Linear}
    sourceBone {boneName=upperarm_r}
    targetBone {boneName=shoulderArmor_r_offset}
   }
   n11 = Get OnGroundShoulderCorrectiveAlpha() -> (OnGroundShoulderCorrectiveAlpha:double)
   n12 = "shoulderArmor_l_mstr.rotateX = upperarm_twist_01_l.rotateX * 2.15 - Bone Driven Controller"!(ComponentPose:Component Space Pose Link Structure=@n10, Alpha:float=@n11) -> (Pose:Component Space Pose Link Structure) {
    is=title AnimGraphNode_BoneDrivenController
    bAffectTargetRotationX=true
    lODThreshold=2
    multiplier=2.15
    sourceComponent=RotationX
    alphaBoolBlend {blendOption=Linear}
    sourceBone {boneName=upperarm_twist_01_l}
    targetBone {boneName=shoulderArmor_l_mstr}
   }
   n13 = "Transform (Modify) Bone - Bone: shoulderArmor_l_mstr"!(ComponentPose=@n12, Alpha=@n11, Translation="0,-0.75,1.3") -> (Pose) {
    translationMode=BMM_Additive
    translationSpace=BCS_BoneSpace
    alphaBoolBlend {blendOption=Linear}
    boneToModify {boneName=shoulderArmor_l_mstr}
    translation {y=-0.75, z=1.3}
   }
   n14 = Get PawnSpeed2D()
   n15 = Anim Dynamics - Bone: shoulderArmor_l_dyn(ComponentPose=@n13, Alpha=@n14, ExternalForce="0,0,0") {
    bGravityOverrideInSimSpace=true
    gravityScale=0.25
    lODThreshold=1
    alphaBoolBlend {blendOption=Linear}
    alphaScaleBiasClamp {bClampResult=true, clampMin=0.25, scale=0.01}
    boundBone {boneName=shoulderArmor_l_dyn}
    physicsBodyDefinitions [{
      sphereCollisionRadius=0
      boundBone {boneName=shoulderArmor_l_dyn}
      boxExtents=<5 5 5>
      constraintSetup {angularLimitsMax=<30 15 0>, angularLimitsMin=<-30 -15 0>, angularTarget=<0 0 0>, linearAxesMax=<0.25 0 0.25>, linearAxesMin=<-0.25 0 -0.25>}
      localJointOffset=<0.5 0 0>
     }]
   }
   n16 = Anim Dynamics - Bone: shoulderArmor_r_dyn(ComponentPose=@n15, Alpha=@n14, ExternalForce="0,0,0") {
    bGravityOverrideInSimSpace=true
    gravityScale=0.25
    lODThreshold=1
    alphaBoolBlend {blendOption=Linear}
    alphaScaleBiasClamp {bClampResult=true, clampMin=0.25, scale=0.01}
    boundBone {boneName=shoulderArmor_r_dyn}
    physicsBodyDefinitions [{
      sphereCollisionRadius=0
      boundBone {boneName=shoulderArmor_r_dyn}
      boxExtents=<5 5 5>
      constraintSetup {angularLimitsMax=<30 15 0>, angularLimitsMin=<-30 -15 0>, angularTarget=<0 0 0>, linearAxesMax=<0.25 0 0.25>, linearAxesMin=<-0.25 0 -0.25>}
      localJointOffset=<-0.5 0 0>
     }]
   }
   Output Pose(Result=Component To Local(ComponentPose=@n16)) {layerGroup=None}
  ]}
 EventGraph {nodes [
   Event BlueprintUpdateAnimation()
   > n1 = Cast To FortPlayerPawn(Object=TryGetPawnOwner())
   n2 = Break Fort Anim Input Skydiving!(FortAnimInput_Skydiving:"Fort Anim Input Skydiving Structure (by ref)"=GetSkydivingInput!(self:Fort Player Animation Interface Object Reference=GetAnimationInterface!(self:Fort Player Pawn Object Reference=@n1) -> (ReturnValue:Fort Player Animation Interface Object Reference)) -> (ReturnValue:"Fort Anim Input Skydiving Structure (by ref)")) -> (
    ParaGilderRootOffset:Vector=~
    ParaGliderRootRotationOverride:Rotator=~
    bUseParaGlideRootModifier:bool=~
    bIsSkydivingFromLaunchPad:bool=~
    bIsSkydivingFromBus:bool=~
    bIsInVortex:bool=~
    bIsDiving:bool=~
    bIsParachuteOpen:bool=~
    bIsSkydiving:bool=~
    bIsParachuteLeaning:bool=~
    bIsSkydiveLeaning:bool=~
    bIsLeaning:bool=~
    bIsSkydiveDiveMode:bool=~
    bPlaySkydiveDrift:bool=~
    bSkydiveDriftDelayActive:bool=~
    bSkydiveDriftAnimAllowed:bool=~
    bIsGliderRight:bool=~
    bIsGliderCenter:bool=~
    bIsGliderLeft:bool=~
    bIsGliderForward:bool=~
    bIsGliderBack:bool=~
    bHasDeployedAndIsStandingUpright:bool=~
    LocalAccelForward:float=~
    LocalAccelRight:float=~
    LocalVelocityRight:float=~
    SkydiveAimPitch:float=~
    SkydiveAimYaw:float=~
    DeployChuteAnimRate:float=~
    SkydiveDriftAnimRate:float=~
    SkydiveDriftAnimRateCurrent:float=~
    SkydiveFidgetAnimRate:float=~
    SkydiveFidgetAnimRateCurrent:float=~
    SkydiveAdditiveAlpha:float=~
    SkydiveDriftDelay:float=~
    SkydiveDriftAnim:int
    SkydiveDriftAnimMax:int
    LaunchpadAnim:int
    LocalAccelDir:ESkydivingDirection Enum
    DirectionLast:ESkydivingDirection Enum
    PoseAdditive:Anim Sequence Object Reference
   ) {is=title K2Node_BreakStruct}
   > n3 = Branch(Condition=OR Boolean!(A=@n2.bIsSkydivingFromLaunchPad, B=@n2.bIsDiving, C:bool=@n2.bIsParachuteOpen, D:bool=@n2.bIsSkydiving) -> (ReturnValue))
   > Set OnGroundShoulderCorrectiveAlpha(OnGroundShoulderCorrectiveAlpha:double=0) -> (Output_Get:double)
   > Set InAirShoulderCorrectiveAlpha(InAirShoulderCorrectiveAlpha:double=SelectFloat(A=0.5, B=1, bPickA=IsUsingUmbrella!(self:Fort Player Pawn Object Reference=@n1) -> (ReturnValue:bool))) -> (Output_Get:double)
   Set OnGroundShoulderCorrectiveAlpha(execute=@n3.else, OnGroundShoulderCorrectiveAlpha=1)
   > Set InAirShoulderCorrectiveAlpha(InAirShoulderCorrectiveAlpha=0)
   Event BlueprintInitializeAnimation()
   > Set MeshToCopy(MeshToCopy:Skeletal Mesh Component Object Reference=GetOwnerSkeletalMeshComponent()) -> (Output_Get:Skeletal Mesh Component Object Reference)
  ]}
]
variables [
 InAirShoulderCorrectiveAlpha {type=number}
 MeshToCopy {type=SkeletalMeshComponent}
 OnGroundShoulderCorrectiveAlpha {type=number}
]
